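# Sound cue for the MotherWall NPC's voice line on entering the DBNO (down but not out) state.
# A Random node picks one of three equally weighted wave variants and is excluded from branch culling.
# Playback uses the NPC VO sound class, close-range MotherWall attenuation and a chatter concurrency set.
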SoundCue /MotherWall/Audio/VO/MotherWall_DBNO_Enter_Cue
root=Random_0
nodes [
 {id=Random_0, children [WavePlayer_1,WavePlayer_2,WavePlayer_3], properties {bShouldExcludeFromBranchCulling=true, weights [1,1,1]}}
 {id=WavePlayer_1, properties {soundWaveAssetPtr=/MotherWall/Audio/VO/Waves/MotherWall_2021_DBNO_Enter_010}}
 {id=WavePlayer_2, properties {soundWaveAssetPtr=/MotherWall/Audio/VO/Waves/MotherWall_2021_DBNO_Enter_011}}
 {id=WavePlayer_3, properties {soundWaveAssetPtr=/MotherWall/Audio/VO/Waves/MotherWall_2021_DBNO_Enter_012}}
]
properties {
 attenuationSettings=&/MotherWall_Close_Att
 maxDistance=4000
 soundClassObject=&Fort_VO_NPC
 volumeMultiplier=1
 concurrencySet [&/MotherWall_Chatter_Conc]
}
