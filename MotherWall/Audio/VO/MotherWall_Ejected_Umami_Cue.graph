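# MotherWall voice-over cue for Umami's "Ejected" barks, playing one of four recorded line variations.
# A Random node with equal weights picks among the four wave players and is excluded from branch culling.
# The cue routes through the NPC VO sound class with boosted volume, close-range Umami attenuation, and shared MotherWall chatter concurrency to limit overlapping lines.
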SoundCue /MotherWall/Audio/VO/MotherWall_Ejected_Umami_Cue
root=Random_1
nodes [
 {
  id=Random_1
  children [WavePlayer_1,WavePlayer_6,WavePlayer_7,WavePlayer_8]
  properties {bShouldExcludeFromBranchCulling=true, weights [1,1,1,1]}
 }
 {id=WavePlayer_1, properties {soundWaveAssetPtr=/MotherWall/Audio/VO/Waves/MotherWall_Umami_Ejected_01}}
 {id=WavePlayer_6, properties {soundWaveAssetPtr=/MotherWall/Audio/VO/Waves/MotherWall_Umami_Ejected_02}}
 {id=WavePlayer_7, properties {soundWaveAssetPtr=/MotherWall/Audio/VO/Waves/MotherWall_Umami_Ejected_03}}
 {id=WavePlayer_8, properties {soundWaveAssetPtr=/MotherWall/Audio/VO/Waves/MotherWall_Umami_Ejected_04}}
]
properties {
 attenuationSettings=&/MotherWall_Umami_Close_Att
 maxDistance=13000
 soundClassObject=&Fort_VO_NPC
 volumeMultiplier=1.5
 concurrencySet [&/MotherWall_Chatter_Conc]
}
