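# Sound cue for the MotherWall Umami voice's full-alert idle chatter, classed as NPC VO.
# A Random node holds five equally weighted wave variations plus four empty, higher-weighted slots, and is excluded from branch culling.
# A Quality Level root routes its first two levels to that same node and leaves the other two empty.
# An Umami idle attenuation asset handles falloff, and a MotherWall chatter concurrency group governs overlapping playback.
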SoundCue /MotherWall/Audio/VO/MotherWall_FullAlertIdle_Umami_Cue
root=QualityLevel_0
nodes [
 {id=QualityLevel_0, children [Random_0,Random_0,null,null]}
 {
  id=Random_0
  children [WavePlayer_0,WavePlayer_1,WavePlayer_2,WavePlayer_3,WavePlayer_4,null,null,null,null]
  properties {bShouldExcludeFromBranchCulling=true, weights [0.2,0.2,0.2,0.2,0.2,1,1,1,1]}
 }
 {id=WavePlayer_0, properties {soundWaveAssetPtr=/MotherWall/Audio/VO/Waves/MotherWall_Umami_FullAlertIdle_01}}
 {id=WavePlayer_1, properties {soundWaveAssetPtr=/MotherWall/Audio/VO/Waves/MotherWall_Umami_FullAlertIdle_02}}
 {id=WavePlayer_2, properties {soundWaveAssetPtr=/MotherWall/Audio/VO/Waves/MotherWall_Umami_FullAlertIdle_03}}
 {id=WavePlayer_3, properties {soundWaveAssetPtr=/MotherWall/Audio/VO/Waves/MotherWall_Umami_FullAlertIdle_04}}
 {id=WavePlayer_4, properties {soundWaveAssetPtr=/MotherWall/Audio/VO/Waves/MotherWall_Umami_FullAlertIdle_05}}
]
properties {
 attenuationSettings=&/MotherWall_Umami_Idle_Att
 maxDistance=8000
 soundClassObject=&Fort_VO_NPC
 volumeMultiplier=1
 concurrencySet [&/MotherWall_Chatter_Conc]
}
