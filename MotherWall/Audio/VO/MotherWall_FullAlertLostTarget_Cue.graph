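# Sound cue for MotherWall voice lines, named for losing its target while in full alert.
# A random node picks among four equally weighted waves: three lost-target lines and one borrowed suspicious-return-to-idle line.
# It plays in the NPC VO sound class with close-range MotherWall attenuation and a chatter concurrency set, and the random node is excluded from branch culling.
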SoundCue /MotherWall/Audio/VO/MotherWall_FullAlertLostTarget_Cue
root=Random_1
nodes [
 {
  id=Random_1
  children [WavePlayer_0,WavePlayer_9,WavePlayer_4,WavePlayer_1]
  properties {bShouldExcludeFromBranchCulling=true, weights [1,1,1,1]}
 }
 {id=WavePlayer_0, properties {soundWaveAssetPtr=/MotherWall/Audio/VO/Waves/MotherWall_2021_FullAlertLostTarget_020}}
 {id=WavePlayer_1, properties {soundWaveAssetPtr=/MotherWall/Audio/VO/Waves/MotherWall_2021_FullAlertLostTarget_030}}
 {id=WavePlayer_4, properties {soundWaveAssetPtr=/MotherWall/Audio/VO/Waves/MotherWall_2021_SuspiciousReturnToIdle_010}}
 {id=WavePlayer_9, properties {soundWaveAssetPtr=/MotherWall/Audio/VO/Waves/MotherWall_2021_FullAlertLostTarget_051}}
]
properties {
 attenuationSettings=&/MotherWall_Close_Att
 maxDistance=4000
 soundClassObject=&Fort_VO_NPC
 volumeMultiplier=1
 concurrencySet [&/MotherWall_Chatter_Conc]
}
