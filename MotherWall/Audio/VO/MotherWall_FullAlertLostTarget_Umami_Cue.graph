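# Sound cue for the MotherWall Umami NPC voice bark that, going by its name, plays when a full-alert character loses its target.
# A Random node picks one of five equally weighted voice-line waves and is excluded from branch culling.
# Output uses the NPC VO sound class, the Umami close attenuation and the MotherWall chatter concurrency group.
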SoundCue /MotherWall/Audio/VO/MotherWall_FullAlertLostTarget_Umami_Cue
root=Random_0
nodes [
 {
  id=Random_0
  children [WavePlayer_2,WavePlayer_3,WavePlayer_5,WavePlayer_6,WavePlayer_7]
  properties {bShouldExcludeFromBranchCulling=true, weights [1,1,1,1,1]}
 }
 {id=WavePlayer_2, properties {soundWaveAssetPtr=/MotherWall/Audio/VO/Waves/MotherWall_Umami_FullAlertLostTarget_01}}
 {id=WavePlayer_3, properties {soundWaveAssetPtr=/MotherWall/Audio/VO/Waves/MotherWall_Umami_FullAlertLostTarget_02}}
 {id=WavePlayer_5, properties {soundWaveAssetPtr=/MotherWall/Audio/VO/Waves/MotherWall_Umami_FullAlertLostTarget_03}}
 {id=WavePlayer_6, properties {soundWaveAssetPtr=/MotherWall/Audio/VO/Waves/MotherWall_Umami_FullAlertLostTarget_04}}
 {id=WavePlayer_7, properties {soundWaveAssetPtr=/MotherWall/Audio/VO/Waves/MotherWall_Umami_FullAlertLostTarget_05}}
]
properties {
 attenuationSettings=&/MotherWall_Umami_Close_Att
 maxDistance=13000
 soundClassObject=&Fort_VO_NPC
 volumeMultiplier=1
 concurrencySet [&/MotherWall_Chatter_Conc]
}
