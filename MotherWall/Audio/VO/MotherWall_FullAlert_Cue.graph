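# Sound cue for MotherWall's full-alert voice lines, played as NPC VO.
# A random node picks with equal weight among four wave players: three FullAlert takes and one FullAlertIdle take.
# The random node is excluded from branch culling.
# Playback uses close-range attenuation, a chatter concurrency group and a boosted volume.
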SoundCue /MotherWall/Audio/VO/MotherWall_FullAlert_Cue
root=Random_1
nodes [
 {
  id=Random_1
  children [WavePlayer_7,WavePlayer_8,WavePlayer_9,WavePlayer_1]
  properties {bShouldExcludeFromBranchCulling=true, weights [1,1,1,1]}
 }
 {id=WavePlayer_1, properties {soundWaveAssetPtr=/MotherWall/Audio/VO/Waves/MotherWall_2021_FullAlertIdle_010}}
 {id=WavePlayer_7, properties {soundWaveAssetPtr=/MotherWall/Audio/VO/Waves/MotherWall_2021_FullAlert_030}}
 {id=WavePlayer_8, properties {soundWaveAssetPtr=/MotherWall/Audio/VO/Waves/MotherWall_2021_FullAlert_040}}
 {id=WavePlayer_9, properties {soundWaveAssetPtr=/MotherWall/Audio/VO/Waves/MotherWall_2021_FullAlert_050}}
]
properties {
 attenuationSettings=&/MotherWall_Close_Att
 maxDistance=4000
 soundClassObject=&Fort_VO_NPC
 volumeMultiplier=1.5
 concurrencySet [&/MotherWall_Chatter_Conc]
}
