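# Sound cue for Umami's full-alert voice line in MotherWall, played as NPC VO with close attenuation and chatter concurrency.
# A quality-level node sends its first two levels to a random pick of four waves and its last two to a pick of two.
# Both random nodes include equally weighted empty slots, so some triggers play nothing, and both are excluded from branch culling.
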SoundCue /MotherWall/Audio/VO/MotherWall_FullAlert_Umami_Cue
root=QualityLevel_0
nodes [
 {id=QualityLevel_0, children [Random_0,Random_0,Random_1,Random_1]}
 {
  id=Random_0
  children [WavePlayer_0,WavePlayer_3,WavePlayer_4,WavePlayer_5,null,null,null]
  properties {bShouldExcludeFromBranchCulling=true, weights [1,1,1,1,1,1,1]}
 }
 {id=Random_1, children [WavePlayer_1,WavePlayer_2,null,null], properties {bShouldExcludeFromBranchCulling=true, weights [1,1,1,1]}}
 {id=WavePlayer_0, properties {soundWaveAssetPtr=/MotherWall/Audio/VO/Waves/MotherWall_Umami_FullAlert_01}}
 {id=WavePlayer_1, properties {soundWaveAssetPtr=&MotherWall_Umami_FullAlert_01}}
 {id=WavePlayer_2, properties {soundWaveAssetPtr=/MotherWall/Audio/VO/Waves/MotherWall_Umami_FullAlert_03}}
 {id=WavePlayer_3, properties {soundWaveAssetPtr=&MotherWall_Umami_FullAlert_03}}
 {id=WavePlayer_4, properties {soundWaveAssetPtr=/MotherWall/Audio/VO/Waves/MotherWall_Umami_FullAlert_04}}
 {id=WavePlayer_5, properties {soundWaveAssetPtr=/MotherWall/Audio/VO/Waves/MotherWall_Umami_FullAlert_05}}
]
properties {
 attenuationSettings=&/MotherWall_Umami_Close_Att
 maxDistance=13000
 soundClassObject=&Fort_VO_NPC
 volumeMultiplier=1.5
 concurrencySet [&/MotherWall_Chatter_Conc]
}
