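# Sound Cue for MotherWall's Umami idle voice lines, played as NPC chatter.
# A Random node picks one of five equally weighted idle wave variants and is excluded from branch culling.
# Playback routes through the NPC VO sound class, with a dedicated idle attenuation asset and a shared chatter concurrency set limiting overlapping lines.
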SoundCue /MotherWall/Audio/VO/MotherWall_Idle_Umami_Cue
root=Random_0
nodes [
 {
  id=Random_0
  children [WavePlayer_1,WavePlayer_2,WavePlayer_3,WavePlayer_4,WavePlayer_5]
  properties {bShouldExcludeFromBranchCulling=true, weights [1,1,1,1,1]}
 }
 {id=WavePlayer_1, properties {soundWaveAssetPtr=/MotherWall/Audio/VO/Waves/MotherWall_Umami_Idle_01}}
 {id=WavePlayer_2, properties {soundWaveAssetPtr=/MotherWall/Audio/VO/Waves/MotherWall_Umami_Idle_02}}
 {id=WavePlayer_3, properties {soundWaveAssetPtr=/MotherWall/Audio/VO/Waves/MotherWall_Umami_Idle_03}}
 {id=WavePlayer_4, properties {soundWaveAssetPtr=/MotherWall/Audio/VO/Waves/MotherWall_Umami_Idle_04}}
 {id=WavePlayer_5, properties {soundWaveAssetPtr=/MotherWall/Audio/VO/Waves/MotherWall_Umami_Idle_05}}
]
properties {
 attenuationSettings=&/MotherWall_Umami_Idle_Att
 maxDistance=8000
 soundClassObject=&Fort_VO_NPC
 volumeMultiplier=1
 concurrencySet [&/MotherWall_Chatter_Conc]
}
