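# Sound cue for the MotherWall Umami NPC voice line played, by its name, when it returns from suspicious to idle.
# A Random node picks one of three equally weighted wave variants and is excluded from branch culling.
# It plays in the NPC VO sound class with Umami close-range attenuation and shares the MotherWall chatter concurrency group to limit overlapping barks.
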SoundCue /MotherWall/Audio/VO/MotherWall_SuspiciousReturnToIdle_Umami_Cue
root=Random_2
nodes [
 {id=Random_2, children [WavePlayer_11, WavePlayer_12, WavePlayer_13], properties {bShouldExcludeFromBranchCulling=true, weights [1,1,1]}}
 {id=WavePlayer_11, properties {soundWaveAssetPtr=/MotherWall/Audio/VO/Waves/MotherWall_Umami_SuspiciousReturnToIdle_04}}
 {id=WavePlayer_12, properties {soundWaveAssetPtr=/MotherWall/Audio/VO/Waves/MotherWall_Umami_SuspiciousReturnToIdle_02}}
 {id=WavePlayer_13, properties {soundWaveAssetPtr=/MotherWall/Audio/VO/Waves/MotherWall_Umami_SuspiciousReturnToIdle_03}}
]
properties {
 attenuationSettings=&/MotherWall_Umami_Close_Att
 maxDistance=13000
 soundClassObject=&Fort_VO_NPC
 volumeMultiplier=1
 concurrencySet [&/MotherWall_Chatter_Conc]
}
