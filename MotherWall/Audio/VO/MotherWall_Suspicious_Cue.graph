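# MotherWall suspicious voice cue that plays one randomly chosen VO line for an NPC.
# An equal-weight Random node picks among four wave players: two Suspicious lines, a suspicious-return-to-idle line and a full-alert idle line.
# The cue routes through the NPC VO sound class with close-range attenuation, a slight volume boost and a shared chatter concurrency group.
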SoundCue /MotherWall/Audio/VO/MotherWall_Suspicious_Cue
root=Random_1
nodes [
 {
  id=Random_1
  children [WavePlayer_6,WavePlayer_7,WavePlayer_2,WavePlayer_1]
  properties {bShouldExcludeFromBranchCulling=true, weights [1,1,1,1]}
 }
 {id=WavePlayer_1, properties {soundWaveAssetPtr=/MotherWall/Audio/VO/Waves/MotherWall_2021_SuspiciousReturnToIdle_040}}
 {id=WavePlayer_2, properties {soundWaveAssetPtr=/MotherWall/Audio/VO/Waves/MotherWall_2021_FullAlertIdle_020}}
 {id=WavePlayer_6, properties {soundWaveAssetPtr=/MotherWall/Audio/VO/Waves/MotherWall_2021_Suspicious_020}}
 {id=WavePlayer_7, properties {soundWaveAssetPtr=/MotherWall/Audio/VO/Waves/MotherWall_2021_Suspicious_030}}
]
properties {
 attenuationSettings=&/MotherWall_Close_Att
 maxDistance=4000
 soundClassObject=&Fort_VO_NPC
 volumeMultiplier=1.2
 concurrencySet [&/MotherWall_Chatter_Conc]
}
