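# Sound cue for suspicious-state voice lines of the MotherWall Umami NPC.
# A Random node picks among six equally weighted Suspicious wave variations and is excluded from branch culling.
# Playback uses the NPC voice sound class, a close-range Umami attenuation asset and a chatter concurrency set, with a slight volume boost.
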SoundCue /MotherWall/Audio/VO/MotherWall_Suspicious_Umami_Cue
root=Random_0
nodes [
 {
  id=Random_0
  children [WavePlayer_0,WavePlayer_3,WavePlayer_4,WavePlayer_5,WavePlayer_8,WavePlayer_9]
  properties {bShouldExcludeFromBranchCulling=true, weights [1,1,1,1,1,1]}
 }
 {id=WavePlayer_0, properties {soundWaveAssetPtr=/MotherWall/Audio/VO/Waves/MotherWall_Umami_Suspicious_02}}
 {id=WavePlayer_3, properties {soundWaveAssetPtr=/MotherWall/Audio/VO/Waves/MotherWall_Umami_Suspicious_03}}
 {id=WavePlayer_4, properties {soundWaveAssetPtr=/MotherWall/Audio/VO/Waves/MotherWall_Umami_Suspicious_04}}
 {id=WavePlayer_5, properties {soundWaveAssetPtr=/MotherWall/Audio/VO/Waves/MotherWall_Umami_Suspicious_05}}
 {id=WavePlayer_8, properties {soundWaveAssetPtr=/MotherWall/Audio/VO/Waves/MotherWall_Umami_Suspicious_06}}
 {id=WavePlayer_9, properties {soundWaveAssetPtr=/MotherWall/Audio/VO/Waves/MotherWall_Umami_Suspicious_07}}
]
properties {
 attenuationSettings=&/MotherWall_Umami_Close_Att
 maxDistance=13000
 soundClassObject=&Fort_VO_NPC
 volumeMultiplier=1.2
 concurrencySet [&/MotherWall_Chatter_Conc]
}
